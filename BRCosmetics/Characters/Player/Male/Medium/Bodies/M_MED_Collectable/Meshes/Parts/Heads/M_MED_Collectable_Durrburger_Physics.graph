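# Physics asset for the Durrburger head part of the male medium Collectable player body, covering its dangling cheese, top and tongue bones.
# Capsules cover the head, three cheese bones and the small top, while boxes cover the three tongue segments.
# Cheese pieces, top and first tongue segment are constrained directly to the head, and the remaining tongue segments chain from it.
# Angular limits keep the top tightest, with looser swing on cheese and tongue.
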PhysicsAsset /BRCosmetics/Characters/Player/Male/Medium/Bodies/M_MED_Collectable/Meshes/Parts/Heads/M_MED_Collectable_Durrburger_Physics
bodies {
 dyn_cheese_bk [{shapeType=Capsule, length=2.75, radius=1, shapeName=dyn_cheese_bk_capsule, center=<2 0 0>, rotation=<rot 90 -177.1056 -177.1059>}]
 dyn_cheese_l [{shapeType=Capsule, length=2.75, radius=1, shapeName=dyn_cheese_l_capsule, center=<2 0 0>, rotation=<rot 90 -157.5511 -157.5515>}]
 dyn_cheese_r [{shapeType=Capsule, length=2.75, radius=1, shapeName=dyn_cheese_r_capsule, center=<2 0 0>, rotation=<rot 90 28.7656 28.766>}]
 dyn_small_tongue_01 [{shapeType=Box, extentX=5, extentY=13.5, extentZ=2.5, shapeName=dyn_small_tongue_01_box, center=<3.6 0 0>}]
 dyn_small_tongue_02 [{shapeType=Box, extentX=6, extentY=12.5, extentZ=2.5, shapeName=dyn_small_tongue_02_box, center=<3.2 0 0>}]
 dyn_small_tongue_03 [{shapeType=Box, extentX=6, extentY=12, extentZ=2.5, shapeName=dyn_small_tongue_03_box, center=<3 0 0>}]
 dyn_small_top [{shapeType=Capsule, length=7.5, radius=3.75, shapeName=dyn_small_top_capsule, center=<2 0 0>, rotation=<rot 100 0 0>}]
 head [{shapeType=Capsule, length=1, radius=1}]
}
constraints [
 {bone1Name=dyn_cheese_r, bone2Name=head, swing1LimitDegrees=30, swing2LimitDegrees=10, twistLimitDegrees=5}
 {bone1Name=dyn_cheese_l, bone2Name=head, swing1LimitDegrees=10, swing2LimitDegrees=30, twistLimitDegrees=5}
 {bone1Name=dyn_cheese_bk, bone2Name=head, swing1LimitDegrees=20, swing2LimitDegrees=30, twistLimitDegrees=5}
 {bone1Name=dyn_small_top, bone2Name=head, swing1LimitDegrees=2, swing2LimitDegrees=10, twistLimitDegrees=2}
 {bone1Name=dyn_small_tongue_01, bone2Name=head, swing1LimitDegrees=30, swing2LimitDegrees=30, twistLimitDegrees=10}
 {bone1Name=dyn_small_tongue_02, bone2Name=dyn_small_tongue_01, swing1LimitDegrees=30, swing2LimitDegrees=30, twistLimitDegrees=10}
 {bone1Name=dyn_small_tongue_03, bone2Name=dyn_small_tongue_02, swing1LimitDegrees=30, swing2LimitDegrees=30, twistLimitDegrees=10}
]
