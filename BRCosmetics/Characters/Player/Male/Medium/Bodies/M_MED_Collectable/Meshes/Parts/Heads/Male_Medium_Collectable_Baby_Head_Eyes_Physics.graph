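# Physics asset for the eye parts of the male medium Collectable baby head, letting the eyelids move as physics bodies.
# Capsule bodies cover the head, each eyelid and a small master pivot per lid, chained head to master to lid on both sides.
# Master joints rotate freely on the head, while each lid is held to almost pure twist within a limited arc, forming a hinge-like joint.
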PhysicsAsset /BRCosmetics/Characters/Player/Male/Medium/Bodies/M_MED_Collectable/Meshes/Parts/Heads/Male_Medium_Collectable_Baby_Head_Eyes_Physics
bodies {
 L_eye_lid [{shapeType=Capsule, length=3.2486, radius=3.1757, shapeName=L_eye_lid_capsule, center=<0 -0.6 0>, rotation=<rot 0 -5.9381 -47.6293>}]
 L_eye_lid_MSTR [{shapeType=Capsule, length=0.505, radius=0.505, shapeName=L_eye_lid_MSTR_capsule, rotation=<rot 0 -90 0>}]
 R_eye_lid [{shapeType=Capsule, length=3.4159, radius=3.0126, shapeName=R_eye_lid_capsule, rotation=<rot 0 6.0119 -47.6131>}]
 R_eye_lid_MSTR [{shapeType=Capsule, length=0.505, radius=0.505, shapeName=R_eye_lid_MSTR_capsule, rotation=<rot 0 -90 0>}]
 head [{shapeType=Capsule, length=10.681, radius=5.0795, center=<8.6189 4.2832 0>}]
}
constraints [
 {
  bone1Name=R_eye_lid_MSTR
  bone2Name=head
  swing1LimitDegrees=45
  swing1Motion=Free
  swing2LimitDegrees=45
  swing2Motion=Free
  twistLimitDegrees=45
  twistMotion=Free
 }
 {
  bone1Name=L_eye_lid_MSTR
  bone2Name=head
  swing1LimitDegrees=45
  swing1Motion=Free
  swing2LimitDegrees=45
  swing2Motion=Free
  twistLimitDegrees=45
  twistMotion=Free
 }
 {bone1Name=L_eye_lid, bone2Name=L_eye_lid_MSTR, swing1LimitDegrees=1, swing2LimitDegrees=1, twistLimitDegrees=35}
 {bone1Name=R_eye_lid, bone2Name=R_eye_lid_MSTR, swing1LimitDegrees=1, swing2LimitDegrees=1, twistLimitDegrees=35}
]
