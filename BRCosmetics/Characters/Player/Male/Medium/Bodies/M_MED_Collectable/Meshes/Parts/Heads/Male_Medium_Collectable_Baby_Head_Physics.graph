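# Physics asset for the head part of the male medium Collectable Baby character, covering the head and its dynamic hair bones.
# Capsule bodies sit on the head and on two two-bone hair chains, left and right.
# Each chain root is constrained to the head with narrow swing limits, and each tip to its root with wider swing, while twist stays tight throughout.
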PhysicsAsset /BRCosmetics/Characters/Player/Male/Medium/Bodies/M_MED_Collectable/Meshes/Parts/Heads/Male_Medium_Collectable_Baby_Head_Physics
bodies {
 dyn_hair_l [{shapeType=Capsule, length=6.5525, radius=2.6664, shapeName=dyn_hair_l_capsule, center=<2 0 3>, rotation=<rot 90 0 0>}]
 dyn_hair_l_01 [{shapeType=Capsule, length=6.2972, radius=3.6819, shapeName=dyn_hair_l_01_capsule, center=<3 0 2>, rotation=<rot 90 0 0>}]
 dyn_hair_r [{shapeType=Capsule, length=3.869, radius=2.9808, shapeName=dyn_hair_r_capsule, center=<2 0 3>, rotation=<rot 90 0 0>}]
 dyn_hair_r_01 [{shapeType=Capsule, length=3.6496, radius=2.3867, shapeName=dyn_hair_r_01_capsule, center=<2 0 2.5>, rotation=<rot -45 0 0>}]
 head [{shapeType=Capsule, length=4.9474, radius=8.5946, center=<4.315 1.8096 0>, rotation=<rot 90 180 -180>}]
}
constraints [
 {bone1Name=dyn_hair_r, bone2Name=head, swing1LimitDegrees=15, swing2LimitDegrees=15, twistLimitDegrees=5}
 {bone1Name=dyn_hair_r_01, bone2Name=dyn_hair_r, swing1LimitDegrees=35, swing2LimitDegrees=35, twistLimitDegrees=5}
 {bone1Name=dyn_hair_l, bone2Name=head, swing1LimitDegrees=15, swing2LimitDegrees=15, twistLimitDegrees=5}
 {bone1Name=dyn_hair_l_01, bone2Name=dyn_hair_l, swing1LimitDegrees=35, swing2LimitDegrees=35, twistLimitDegrees=5}
]
